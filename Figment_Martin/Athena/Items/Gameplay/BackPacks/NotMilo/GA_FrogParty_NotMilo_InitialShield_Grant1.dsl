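; Gameplay ability for the FrogParty NotMilo backpack that applies a heal-up effect, apparently as part of an initial shield grant.
; On activation it gets the avatar actor, checks that it is valid and reads its location.
; It applies GE_HealUp to the owner only when the avatar's Z is at or above a fixed high-altitude threshold.
; An embedded comment describes a hero-tag check and delayed heal, but neither appears in this graph.
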(event Ability|EventActivateAbility
  (bind _returnvalue (Ability|GetAvatarActorfromActorInfo))
  (bind (_returnvalue_x _returnvalue_y _returnvalue_z) (Math|Vector|BreakVector (Transformation|GetActorLocation _returnvalue)))
  ;[1 1 1] Give players with the hero tag a GE for health and shield in addition to giving them a heal a frame later to make sure that they get the full shield.
  (Utilities|IsValid _returnvalue
    (:Is%20Valid
      (if (>= _returnvalue_z 5e+04)
        (Ability|ApplyGameplayEffectToOwner (Variables|Default|GetGE_HealUp))))))
