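; Gameplay ability for the NotMilo backpack that grants health and shield to players carrying the hero tag.
; On activation it validates the avatar actor, casts it to a Fort player pawn and checks its tags against Tag_Hero.
; A match gets GE_HealthShield, then GE_HealUp after a 0.01-second delay so the shield ends up full.
; The graph has no End Ability call.
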(event Ability|EventActivateAbility
  (bind _returnvalue (Ability|GetAvatarActorfromActorInfo))
  ;[1 1 1] Give players with the hero tag a GE for health and shield in addition to giving them a heal a frame later to make sure that they get the full shield.
  (Utilities|IsValid _returnvalue
    (:Is%20Valid
      (bind _asfort_player_pawn (Utilities|Casting|CastToFortPlayerPawn _returnvalue))
      (if (GameplayTags|TagContainer|CompareTagAssetInterfacetoOtherTags (Variables|Default|GetTag_Hero) _asfort_player_pawn)
        (Ability|ApplyGameplayEffectToOwner (Variables|Default|GetGE_HealthShield))
        (Utilities|FlowControl|Delay 0.01)
        (Ability|ApplyGameplayEffectToOwner (Variables|Default|GetGE_HealUp))))))
